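; Vehicle cosmetic actor for the ElectroBeam drift trail, which tints its Niagara effect to the selected paint color.
; When the color is ready, it looks up that name in DT_VehicleCosmetics_Painted and feeds the row's color to the LightningColor and OrbsColor variables on NS_DriftTrail.
; It also calls SetupPaintSettings, whose override only forwards to the parent.
; BeginPlay, Tick, overlap and construction add no logic of their own.
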(event EventBeginPlay
  (|CallParentFunction) ; This node is disabled and will not be called.\nDrag off pins to build functionality.
)

(event Collision|EventActorBeginOverlap
  (params (OtherActor :type Actor%20Object%20Reference)))

(event EventTick
  (params (DeltaSeconds :type Float%20%28single-precision%29)))

(event Vehicle|Cosmetics|EventOnColorReady
  (params (Color :type Name :default "None") (AdditionalInfo :type Vehicle%20Cosmetics%20Additional%20Parameter%20Infos%20Structure))
  (bind _ns_drifttrail (Variables|Default|GetNS_DriftTrail))
  (bind _returnvalue (Utilities|GetDataTableRow "/VehicleCosmetics/DataTables/DT_VehicleCosmetics_Painted.DT_VehicleCosmetics_Painted" Color)
    (:then
      (Niagara|SetNiagaraVariable(LinearColor) _ns_drifttrail "LightningColor" (Utilities|Struct|BreakStructVehicleCosmeticsPainted _returnvalue))
      (Niagara|SetNiagaraVariable(LinearColor) _ns_drifttrail "OrbsColor" (Utilities|Struct|BreakStructVehicleCosmeticsPainted _returnvalue)))
    (:RowNotFound
      _))
  (CallFunction|SetupPaintSettings Color AdditionalInfo))

(fn SetupPaintSettings
  (params (Color :type Name :default "None") (AdditionalInfo :type Vehicle%20Cosmetics%20Additional%20Parameter%20Infos%20Structure))
  (bind (_primarycolor _balanced _metallic) (|CallParentFunction Color AdditionalInfo))
  (return _primarycolor _balanced _metallic))

(fn ConstructionScript
  (|CallParentFunction))
