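# Translucent, two-sided unlit Niagara material for the electricity in the ElectroBeam drift trail cosmetic.
# UV1 pans and periodically jumps on a time-driven sine step, driving both clamped caustics noise for particle-tinted emissive and the lightning texture for opacity.
# IsPainted switches the noise's minimum clamp; particle alpha, depth fade and a UV0 U ramp shape opacity.
# The second flicker timer is inert because its lerp blends identical inputs.
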Material /VehicleCosmetics/DriftTrail/DriftTrail_ElectroBeam/Materials/Electricity_MAT_
graph {
 properties {
  bAutomaticallySetUsageInEditor=false
  bUsedWithGeometryCache=true
  bUsedWithNiagaraMeshParticles=true
  bUsedWithNiagaraRibbons=true
  bUsedWithNiagaraSprites=true
  bUsedWithStaticMesh=false
  blendMode=BLEND_Translucent
  pixelDepthOffsetMode=PDOM_Legacy
  shadingModel=MSM_Unlit
  twoSided=true
  usedShadingModels=Unlit
 }
 nodes [
  n1 = Multiply(A=Time(), B=ScalarParameter() {defaultValue=1, parameterName=OffsetSpeed})
  n2 = ScalarParameter() {defaultValue=0.3, parameterName=OffsetPeriod}
  n3 = Multiply(A=Add(A=Fmod(A=@n1, B=@n2), B=Constant() {r=0.15}), B=Constant() {r=3})
  n4 = Ceil(None=Sine(None=Multiply(A=@n3, B=@n3)))
  n5 = ScalarParameter() {defaultValue=0.25, parameterName=U_OffsetMult}
  n6 = TextureCoordinate() {coordinateIndex=1}
  n7 = ComponentMask(None=VectorParameter() {parameterName=TilingXY, defaultValue=<rgba 1 1 0 1>}.RGB) {g=true, r=true}
  n8 = Time()
  n9 = ScalarParameter() {defaultValue=-1, parameterName=PanSpeed}
  n10 = Add(A=AppendVector(A=Multiply(A=@n4, B=@n5), B=Constant()), B=Panner(Coordinate=Multiply(A=@n6, B=@n7), Time=Multiply(A=@n8, B=@n9)) {speedX=-1})
  n11 = TextureSample(UVs=Multiply(A=@n10, B=Constant2Vector() {g=2, r=2})) {texture=/VehicleCosmetics/DriftTrail/DriftTrail_ElectroBeam/Textures/CausticsNoise}
  n12 = ScalarParameter() {parameterName=IsPainted}
  n13 = Constant() {r=0.1}
  n14 = If(A=Saturate(None=Floor(None=@n12)), "A > B"=@n13, "A == B"=@n13, "A < B"=ScalarParameter() {defaultValue=0.25, parameterName=MaskClampMin}) {constB=0.5}
  n15 = ScalarParameter() {defaultValue=1, parameterName=MaskClampMax}
  MaterialExpressionVertexColor_2 = ParticleColor()
  MaterialFunctionCall(Emissive=Multiply(A=Clamp(None=@n11.RGB, Min=@n14, Max=@n15), B=@MaterialExpressionVertexColor_2.RGB)) {materialFunction=&MF_FX_EmissiveAdjustments}
  n16 = Multiply(A=@n10, B=AppendVector(A=Constant() {r=1}, B=ScalarParameter() {defaultValue=1, parameterName=FlipV}))
  n17 = AppendVector(A=Add(A=ScalarParameter() {parameterName=RandVOffset}, B=Constant()), B=Constant())
  n18 = Constant() {r=1}
  n19 = ScalarParameter() {defaultValue=1, parameterName=ScaleY}
  n20 = TextureSample(UVs=MaterialFunctionCall(
    UVs=Add(A=@n16, B=@n17)
    Texture Scale=MaterialFunctionCall(X=@n18, Y=@n19) {materialFunction=&MakeFloat2}.Result
   ) {materialFunction=&ScaleUVsByCenter}.UVs) {samplerType=SAMPLERTYPE_LinearColor, texture=/DealershipSharedLibrary/Assets/Sets/ElectroBeam/ElectroBeam_01}
  n21 = Multiply(A=Time(), B=ScalarParameter() {defaultValue=0.325, parameterName=FlipSpeed})
  n22 = ScalarParameter() {defaultValue=0.4, parameterName=Period}
  n23 = Multiply(A=Add(A=Fmod(A=@n21, B=@n22), B=Constant() {r=0.15}), B=Constant() {r=3})
  n24 = Sine(None=Multiply(A=@n23, B=@n23))
  n25 = Power(Base=LinearInterpolate(A=@n20.R, B=@n20.R, Alpha=Ceil(None=@n24)), Exp=ScalarParameter() {parameterName=Power Lightning})
  MaterialExpressionConstantClamp_5 = Saturate(None=DepthFade(Opacity=Multiply(A=@MaterialExpressionVertexColor_2.A, B=@n25), FadeDistance=Constant() {r=8}))
  n26 = ComponentMask(None=TextureCoordinate()) {r=true}
  Multiply(A=@MaterialExpressionConstantClamp_5, B=Multiply(A=Power(Base=@n26) {constExponent=1.5}) {constB=3})
 ]
}
