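# Material function for the ElectroBeam drift trail that builds an animated three-component wave vector, exposed to the material library.
# Tiled UVs scroll by Time times Speed and shift by FlickerOffset, then feed one sine and two cosines with differing periods, the cosines adding constant phase offsets.
# Only the U-driven red channel of each wave is kept, appended into XYZ and multiplied by per-axis ScaleX/Y/Z inputs.
# A trailing Round node is unused by the output.
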MaterialFunction /VehicleCosmetics/DriftTrail/DriftTrail_ElectroBeam/Materials/ElectroTrailNoise_MF
graph {
 properties {bExposeToLibrary=true, previewBlendMode=BLEND_Translucent}
 nodes [
  n1 = FunctionInput() {bUsePreviewValueAsDefault=true, inputName=Tile, inputType=FunctionInput_Scalar, previewValue=<1 0 0 1>}
  n2 = Multiply(A=TextureCoordinate(), B=@n1)
  n3 = FunctionInput() {bUsePreviewValueAsDefault=true, inputName=Speed, inputType=FunctionInput_Scalar, previewValue=<1 0 0 1>}
  n4 = Multiply(A=@n3, B=Time())
  n5 = Add(A=@n2, B=@n4)
  n6 = FunctionInput() {bUsePreviewValueAsDefault=true, inputName=FlickerOffset, inputType=FunctionInput_Scalar, previewValue=<1 0 0 1>}
  n7 = ComponentMask(None=Sine(None=Add(A=@n5, B=@n6)) {period=0.5}) {r=true}
  n8 = Add(A=Add(A=@n4, B=Constant() {r=4.234}), B=@n2)
  n9 = ComponentMask(None=Cosine(None=Add(A=@n8, B=@n6))) {r=true}
  n10 = Add(A=@n4, B=Constant() {r=1.181})
  n11 = Cosine(None=Add(A=Add(A=@n2, B=@n10), B=@n6)) {period=0.75}
  n12 = FunctionInput() {bUsePreviewValueAsDefault=true, inputName=ScaleX, inputType=FunctionInput_Scalar, previewValue=<1 1 1 1>}
  n13 = FunctionInput() {bUsePreviewValueAsDefault=true, inputName=ScaleY, inputType=FunctionInput_Scalar, previewValue=<1 1 1 1>}
  n14 = FunctionInput() {bUsePreviewValueAsDefault=true, inputName=ScaleZ, inputType=FunctionInput_Scalar, previewValue=<1 1 1 1>}
  n15 = Multiply(A=AppendVector(A=AppendVector(A=@n7, B=@n9), B=ComponentMask(None=@n11) {r=true}), B=AppendVector(A=AppendVector(A=@n12, B=@n13), B=@n14))
  FunctionOutput(None=@n15)
  Round(None=@n15)
 ]
}
