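# Translucent unlit Niagara ribbon and mesh-particle material for the ElectroBeam vehicle drift trail.
# Emissive blends bright and dark colors across V, tinted by particle color; IsWheelWrap adds panning lightning with time-stepped UV jumps, driven by dynamic parameters.
# An optional switch enables flickering ElectroTrailNoise world position offset, masked along UV1.
# Opacity combines particle alpha, camera-proximity fade and depth fade, plus V-edge fade for wheel wrap.
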Material /VehicleCosmetics/DriftTrail/DriftTrail_ElectroBeam/Materials/Ribbon_MAT
graph {
 properties {
  bUsedWithNiagaraMeshParticles=true
  bUsedWithNiagaraRibbons=true
  blendMode=BLEND_Translucent
  pixelDepthOffsetMode=PDOM_Legacy
  shadingModel=MSM_Unlit
  usedShadingModels=Unlit
 }
 nodes [
  n1 = DynamicParameter() {paramNames [Lerp, UVXNoise, IntensityNoise, ContrastNoise]}
  n2 = Constant() {r=1}
  n3 = Constant() {r=1}
  n4 = Multiply(A=Time(), B=ScalarParameter() {parameterName=OffsetSpeed})
  n5 = ScalarParameter() {defaultValue=0.3, parameterName=OffsetPeriod}
  n6 = Multiply(A=Add(A=Fmod(A=@n4, B=@n5), B=Constant() {r=0.15}), B=Constant() {r=3})
  n7 = Sine(None=Multiply(A=@n6, B=@n6))
  n8 = AppendVector(A=Multiply(A=Ceil(None=@n7), B=ScalarParameter() {defaultValue=0.25, parameterName=U_OffsetMult}), B=Constant())
  n9 = VectorParameter() {parameterName=TilingXY, defaultValue=<rgba 1 1 0 1>}
  n10 = Panner(
   Coordinate=Multiply(A=TextureCoordinate() {coordinateIndex=1}, B=ComponentMask(None=@n9.RGB) {g=true, r=true})
   Time=Multiply(A=Time(), B=ScalarParameter() {defaultValue=0.1, parameterName=PanSpeed})
  ) {speedX=-1}
  n11 = Add(A=Multiply(A=Add(A=@n8, B=@n10), B=AppendVector(A=@n2, B=@n3)), B=AppendVector(A=@n1.UVXNoise, B=@n1.UVXNoise))
  n12 = Multiply(
   A=Multiply(A=VectorParameter() {parameterName=Color_Lightning, defaultValue=<rgba 1 0 0.9038 1>}.RGB, B=@n1.ContrastNoise)
   B=Power(Base=TextureSample(UVs=@n11) {samplerType=SAMPLERTYPE_LinearColor, texture=/VehicleCosmetics/DriftTrail/DriftTrail_ElectroBeam/Textures/ElectroBeam_02}.R) {constExponent=2.5}
  )
  n13 = VectorParameter() {parameterName=Color_Bright, defaultValue=<rgba 2 0.2 0 1>}
  n14 = VectorParameter() {parameterName=Color_Dark, defaultValue=<rgba 0.172 0 0.364 1>}
  n15 = ComponentMask(None=TextureCoordinate()) {g=true}
  n16 = OneMinus(None=@n15)
  n17 = Multiply(A=@n1.Lerp, B=Multiply(A=Multiply(A=@n16, B=@n15), B=Constant() {r=4}))
  n18 = LinearInterpolate(A=@n13.R, B=@n14.R, Alpha=@n17)
  n19 = LinearInterpolate(A=@n13.RGB, B=@n14.RGB, Alpha=@n17)
  MaterialExpressionVertexColor_1 = ParticleColor()
  MaterialExpressionConstantClamp_2 = Saturate(None=Multiply(A=StaticSwitchParameter(True=@n18, False=@n19) {parameterName=PreColorDesaturate}, B=@MaterialExpressionVertexColor_1.RGB))
  MaterialFunctionCall(Emissive=StaticSwitchParameter(True=Add(A=@n12, B=@MaterialExpressionConstantClamp_2), False=@MaterialExpressionConstantClamp_2) {parameterName=IsWheelWrap?}) {materialFunction=&MF_FX_EmissiveAdjustments}
  n20 = ScalarParameter() {defaultValue=6, parameterName=Speed}
  n21 = Multiply(A=Time(), B=ScalarParameter() {parameterName=OffsetSpeed})
  n22 = ScalarParameter() {defaultValue=0.3, parameterName=OffsetPeriod}
  n23 = Multiply(A=Add(A=Fmod(A=@n21, B=@n22), B=Constant() {r=0.15}), B=Constant() {r=3})
  n24 = Multiply(A=@n23, B=@n23)
  n25 = Multiply(A=Ceil(None=Sine(None=@n24)), B=ScalarParameter() {defaultValue=0.25, parameterName=U_OffsetMult})
  n26 = ScalarParameter() {defaultValue=1, parameterName=ScaleX}
  n27 = ScalarParameter() {defaultValue=1, parameterName=ScaleY}
  n28 = ScalarParameter() {defaultValue=1, parameterName=ScaleZ}
  n29 = ScalarParameter() {defaultValue=10, parameterName=tile}
  n30 = ComponentMask(None=TextureCoordinate() {coordinateIndex=1}) {r=true}
  n31 = ScalarParameter() {defaultValue=2, parameterName=NoiseMaskMult}
  MaterialExpressionConstantClamp_4 = Saturate(None=Power(Base=Multiply(A=@n30, B=@n31), Exp=ScalarParameter() {defaultValue=0.35, parameterName=NoiseMaskPow}))
  StaticSwitchParameter(
   True=Multiply(
    A=MaterialFunctionCall!(Speed=@n20, FlickerOffset=@n25, ScaleX=@n26, ScaleY=@n27, ScaleZ=@n28, Tile=@n29) {materialFunction=&/ElectroTrailNoise_MF}.Result
    B=@MaterialExpressionConstantClamp_4
   )
   False=Constant()
  ) {defaultValue=true, parameterName=UseWPO?}
  n32 = Sine(None=ComponentMask(None=TextureCoordinate()) {g=true}) {period=2}
  MaterialExpressionCameraWorldPosition_0 = CameraPositionWS()
  n33 = WorldPosition()
  n34 = Subtract(A=@MaterialExpressionCameraWorldPosition_0, B=@n33.XYZ)
  n35 = Constant3Vector() {constant=<rgba 0 0 -1 0>}
  n36 = Multiply(
   A=Distance(A=@MaterialExpressionCameraWorldPosition_0, B=@n33.XYZ)
   B=DotProduct(A=Normalize(VectorInput=@n34), B=Transform(None=@n35) {transformSourceType=TRANSFORMSOURCE_View})
  )
  n37 = Constant() {r=1}
  n38 = Multiply(A=@MaterialExpressionVertexColor_1.A, B=Divide(A=Subtract(A=@n36, B=@n37), B=Constant() {r=50}))
  n39 = Saturate(None=@n38)
  n40 = ComponentMask(None=TextureCoordinate()) {g=true}
  n41 = OneMinus(None=@n40)
  n42 = Multiply(
   A=Power(Base=Multiply(A=@n40, B=@n41), Exp=ScalarParameter() {defaultValue=2, parameterName=Fade Power})
   B=ScalarParameter() {defaultValue=1, parameterName=Fade_Brightness}
  )
  StaticSwitchParameter(True=Multiply(A=Round(None=@n32), B=Multiply(A=@n39, B=@n42)), False=@n39) {parameterName=IsWheelWrap?}
  n43 = Constant() {r=1}
  n44 = ScalarParameter() {defaultValue=100, parameterName=FadeDistance}
  MaterialExpressionConstantClamp_3 = Saturate(None=Multiply(
    A=@n38
    B=MaterialFunctionCall(Opacity In=@n43, DepthFade Distance=@n44) {materialFunction=&MF_VCVFX_DepthFade}.Result
   ))
  StaticSwitchParameter(True=Multiply(A=@MaterialExpressionConstantClamp_3, B=@n42), False=@MaterialExpressionConstantClamp_3) {parameterName=IsWheelWrap?}
 ]
}
